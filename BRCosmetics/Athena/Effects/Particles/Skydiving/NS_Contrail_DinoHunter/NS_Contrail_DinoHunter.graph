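# Niagara skydiving contrail for the DinoHunter cosmetic, combining trails, meshes and feathers on the character's skeleton.
# Four skeletal ribbons spawn during fast descent, fading from hot red.
# Two persistent bone-attached meshes get speed-driven noise through material parameters scrolled by an integrator, while feather meshes shed faster in dives and drift on curl noise.
# Front-end modifiers substitute ParaGlideLeanAlpha for velocity in several inputs; the cloth-piece emitter is disabled.
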NiagaraSystem /BRCosmetics/Athena/Effects/Particles/Skydiving/NS_Contrail_DinoHunter/NS_Contrail_DinoHunter
dependencies {
 usedDataInterfaces [&NiagaraDataInterfaceVector2DCurve, &NiagaraDataInterfaceCurve, &NiagaraDataInterfaceColorCurve]
 usedDynamicInputs []
 usedModules [
  &SystemState
  /Game/Effects/Niagara/Modules/Skydiving/FN_DivingAttributes/v5/FN_DivingAttributes
  &self:SystemSpawnScript.SystemScriptSource.SystemScriptGraph.NiagaraNodeAssignment_2.SetVariables_#1
  &FN_Integrator
  &self:SystemSpawnScript.SystemScriptSource.SystemScriptGraph.NiagaraNodeAssignment_1.SetVariables_#2
  &self:SystemSpawnScript.SystemScriptSource.SystemScriptGraph.NiagaraNodeAssignment_0.SetVariables_#3
 ]
}
emitters {
 Basic_Contrail {
  emitterUpdateScript [
   &EmitterScalability(
    Module.bApplySpawnCountScale=false
    Module.bApplySpawnCountScaleByDistanceFraction=false
    Module.bUseMaxDistance=false
    Module.bUseMinDistance=false
   ) {enabled=false}
   &EmitterLifeCycle(CompleteOnInactive=true) {enabled=false}
   &EmitterState()
   /Game/Effects/Niagara/Modules/Spawn/FN_SpawnContrails(All limbs=false, Global Spawnrate=60) {enabled=false}
   /Game/Effects/Niagara/Modules/AnimTrail/FN_SpawnAnimTrail/FN_SpawnAnimTrail() {enabled=false}
   &SpawnRate(SpawnRate=30, Use Spawn Probability=false) {enabled=false}
   &FN_SpawnRibbon_v2(
    Number of Ribbons=4
    Spawn Mode=::NewEnumerator0
    SpawnRate=HlslExpression() {hlslExpression="(Engine.Owner.Velocity.z < -2600.0f) ? 40 : 0"}
   )
  ]
  particleSpawnScript [
   &FN_InitializeRibbon_v3(Enable Skeletal Position=true, Lifetime=0.7, Ribbon Width=60) {enabled=false}
   &FN_InitializeRibbon_v3(
    Color:LinearColor=<rgba 1 1 1 0.1>
    Enable Skeletal Position=true
    Lifetime=1
    Sampled Position Coordinate Space:ENiagaraCoordinateSpace=::Local
    Sampled Rotation Coordinate Space:ENiagaraCoordinateSpace=::Local
    Sampled Velocity Coordinate Space:ENiagaraCoordinateSpace=::Local
   ) {module=FN_InitializeRibbon_v3001}
   &AddVelocity(Velocity=<0 0 410>) {enabled=false}
   &AddVelocity(Velocity=&MakeCustomVectorFromBool(
     Bool=$Particles.Locked
     True Vector=<0 0 0>
     False Vector=&FN_FrontEndVector3Modifier(Gameplay=<0 0 100>, Front End=<0 0 350>)
    )) {module=AddVelocity001}
   &DynamicMaterialParameters(
    Index 0 Param 2=0.1
    Index 0 Param 3=1
    Index 0 Param 4=1
    Module.Param0WriteEnabled=true
    Module.Param1WriteEnabled=true
    Module.Param2WriteEnabled=true
    Module.Param3WriteEnabled=true
   )
  ]
  particleUpdateScript [
   &Drag(Drag=0.9, Use Linear Drag=true, Use Rotational Drag=false)
   &InheritVelocity(
    Module.Limit Speed=true
    Velocity Limit=10000
    Velocity Scale=&MakeCustomVectorFromBool(Bool=$Particles.Initial.Locked, True Vector=<0 0 0>, False Vector=<-0.03 -0.03 -0.01>)
    Velocity Source=&TransformVector(Vector=$Engine.Owner.Velocity, Source Space=::World, Destination Space=::Local)
   )
   /Game/Effects/Niagara/Modules/Skydiving/FN_DivingAttributes/FN_DivingAttributes() {enabled=false}
   &SolveForcesAndVelocity(Clamp Velocity=false, Limit Acceleration=false)
   &FN_UpdateRibbon(
    Enable Skeletal Position=true
    Initial Color:LinearColor=$Particles.Initial.Color
    Initial RibbonWidth:NiagaraFloat=$Particles.Initial.RibbonWidth
    Sampled Position Coordinate Space:ENiagaraCoordinateSpace=::Local
    Sampled Rotation Coordinate Space:ENiagaraCoordinateSpace=::Local
    Sampled Velocity Coordinate Space:ENiagaraCoordinateSpace=::Local
   )
   &Material(Base Texture Time:NiagaraFloat=&Multiply_Float(A=$Emitter.Age, B=-1)) {enabled=false}
   &ScaleRibbonWidth(Ribbon Width Scale=&FloatFromCurve(FloatCurve=DataInterface() {propertyValues {Curve {keys [
         {arriveTangent=1.9737, interpMode=RCIM_Cubic, leaveTangent=1.9737, tangentMode=RCTM_User, time=0.0058, value=0.9363}
         {arriveTangent=-0.0131, interpMode=RCIM_Cubic, leaveTangent=-0.0131, time=0.4538, value=1.5093}
         <at 0.6635 0.9277>
        ]}}}))
   &Color(
    Color=&ColorFromCurve(ColorCurve=DataInterface() {propertyValues {
       AlphaCurve {keys [<at 0.1 0 RCIM_Cubic>,<at 0.3 1 RCIM_Cubic>,<at 1 0>]}
       BlueCurve {keys [<at 0 5 RCIM_Cubic>,<at 1 1 RCIM_Cubic>]}
       GreenCurve {keys [<at 0 1.0335 RCIM_Cubic>,<at 1 1 RCIM_Cubic>]}
       RedCurve {keys [<at 0 100 RCIM_Cubic>,<at 1 1 RCIM_Cubic>]}
      }})
    Scale Alpha=&FloatFromCurve(
     FloatCurve=DataInterface() {propertyValues {Curve {keys [
         {arriveTangent=2.7939, interpMode=RCIM_Cubic, leaveTangent=2.7939, tangentMode=RCTM_User, time=-0.0046, value=1.0113}
         {arriveTangent=-1.0036, interpMode=RCIM_Cubic, leaveTangent=-4.3444, tangentMode=RCTM_Break, time=0.1387, value=1.232}
         <at 0.6 0.002>
        ]}}}
     Scale Curve=0.8
    )
   )
  ]
  rendererClasses [&NiagaraRibbonRendererProperties]
  renderers [{rendererClass=&NiagaraRibbonRendererProperties}]
 }
 Capse_LocalSpace {
  emitterSpawnScript [&FN_PreserveSkydiveUserVariables() {module=PreserveUserVariables}]
  emitterUpdateScript [
   &EmitterState(
    Inactive Response=::NewEnumerator1
    Life Cycle Mode=::NewEnumerator1
    Loop Behavior=::NewEnumerator1
    Loop Duration Mode=::NewEnumerator1
    UseLoopDelay=false
   )
   &SpawnBurst_Instantaneous(Spawn Count=2, Use Spawn Probability=false)
  ]
  particleSpawnScript [
   /Game/Effects/Niagara/Modules/Groups/FN_SetSpawnGroup_v1/FN_SetSpawnGroup(Emitter Bone Count=4) {enabled=false}
   /Niagara/Modules/Spawn/Initialization/InitializeParticle(
    Module.Write Color=true
    Module.Write Lifetime=true
    Module.Write Mass=true
    Module.Write Material Random=true
    Module.Write Position=false
    Module.Write Scale=true
    Module.Write SpriteRotation=false
    Module.Write SpriteSize=false
    Position=&SimulationPosition()
   )
   /Game/Effects/Niagara/Modules/Groups/FN_SetSpawnGroup(Number of Spawn Groups=2) {module=FN_SetSpawnGroup001}
   &DynamicMaterialParameters(
    Index 1 Param 3=0.25
    Module.Param0WriteEnabled1=true
    Module.Param1WriteEnabled1=true
    Module.Param2WriteEnabled1=true
    Module.Param3WriteEnabled1=true
    Write Parameter Index 0=false
    Write Parameter Index 1=true
   ) {enabled=false}
  ]
  particleUpdateScript [
   /Game/Effects/Niagara/Modules/Skydiving/FN_DivingAttributes/FN_DivingAttributes_v2(
    Bone Index=3
    SkeletalMesh=DataInterface() {propertyValues {
      DefaultMesh=null
      ExcludeBoneName=root
      OptionalComponentClass=null
      PreviewMesh=&M_Med_Soldier_04_Preview
      SkinningMode=SkinOnTheFly
      SoftSourceActor=null
      SourceMode=Default
      UvSetIndex=0
      WholeMeshLOD=-1
      bExcludeBone=false
      bReadDeformedGeometry=true
      bRequireCurrentFrameData=true
      ComponentTags []
      FilteredBones [upperarm_l,upperarm_r,spine_02]
      FilteredSockets [LootEntryPosition]
      MeshUserParameter {parameter {name=None, typeDefHandle {registeredTypeIndex=1}, varData []}}
      SamplingRegions []
     }}
   )
   &ParticleState(Kill Particles When Lifetime Has Elapsed=false)
   &SkeletalMeshLocation(
    Bone / Socket Index=$Particles.SpawnGroup
    Bone Sampling Mode=::NewEnumerator4
    Inherit Velocity=false
    MaskSpawnGroup=false
    Sampled Position Offset=&MaskVectorBySpawnGroup(Masked Vector=<3 5 0>, Passthrough Vector=<-3 5 0>)
    Sampled Position Offset Coordinate Space:ENiagaraMeshTransforms=::NewEnumerator2
    Skeletal Mesh=$User.SkeletalMesh
   )
   /SetVariables(Particles.MeshOrientation=&LerpQuaternion(
     Quaternion A:Quat4f=<0 0 -0.1 1>
     Quaternion B:Quat4f=<0 0 0.1 1>
     Lerp Factor:NiagaraFloat=&MakeFloatFromInt(Int=$Particles.SpawnGroup)
     Enable Spherical Linear Interpolation:NiagaraBool=false
    )) {isSetParameters=true}
   &SampleSkeletalMeshSkeleton(Mesh:NiagaraDataInterfaceSkeletalMesh=$User.SkeletalMesh, Specific Bone Index:NiagaraInt32=$Particles.SpawnGroup) {enabled=false}
   /Niagara/Modules/Spawn/Location/SkeletalMeshSkeletonLocation(Module.Inherit Velocity=false, Module.Use Sampled Bone as Ribbon ID=true) {enabled=false}
   &SolveForcesAndVelocity(Clamp Velocity=false, Limit Acceleration=false) {enabled=false}
   /Game/Effects/Niagara/Modules/Materials/FN_MeshMaterial(
    Amplitude:NiagaraFloat=&Lerp_Float(
     A=140
     B=100
     Alpha=&FN_FrontEndFloatModifier(Gameplay=HlslExpression() {hlslExpression="saturate((abs(Engine.Owner.Velocity.z) -1200) / 4000)"}, Front End=$User.ParaGlideLeanAlpha)
    )
    Falloff:NiagaraFloat=&Lerp_Float(
     A=350
     B=600
     Alpha=&FN_FrontEndFloatModifier(Gameplay=HlslExpression() {hlslExpression="saturate((abs(Engine.Owner.Velocity.z) -1200) / 4000)"}, Front End=$User.ParaGlideLeanAlpha)
    )
    Frequency=350
    Offset:NiagaraFloat=&FN_FrontEndFloatModifier(
     Gameplay=&Multiply_Float(A=$System.FN_Integrator.Value, B=0.3)
     Front End=&Multiply_Float(A=$System.FN_Integrator.Value, B=0.3)
    )
    Scale Noise X Axis:NiagaraFloat=&Lerp_Float(A=0.6, B=1.5, Alpha=HlslExpression() {hlslExpression="saturate((abs(Engine.Owner.Velocity.z) -1200) / 4000)"})
    Scale Noise Y Axis:NiagaraFloat=&Lerp_Float(A=0.6, B=1.1, Alpha=HlslExpression() {hlslExpression="saturate((abs(Engine.Owner.Velocity.z) -1200) / 4000)"})
    Scale Noise Z Axis=0.1
    Spawn Group Count=2
   )
  ]
  rendererClasses [&NiagaraMeshRendererProperties]
  renderers [
   {rendererClass=&NiagaraMeshRendererProperties, rendererIndex=0}
   {rendererClass=&NiagaraMeshRendererProperties, rendererIndex=1}
  ]
 }
 Feathers {
  emitterSpawnScript [&FN_PreserveSkydiveUserVariables() {module=PreserveUserVariables}]
  emitterUpdateScript [
   &EmitterState(Life Cycle Mode=::NewEnumerator1, Loop Duration=2, UseLoopDelay=false)
   &SpawnRate(
    Spawn Probability=0.1
    SpawnRate=HlslExpression() {hlslExpression="(Engine.Owner.Velocity.z < -2600.0f) ? 50.0f : 10.0f"}
    Use Spawn Probability=true
   ) {enabled=false}
   /Game/Effects/Niagara/Modules/Skydiving/FN_DivingAttributes/FN_EmitterDivingAttributes_v1(Accumulate to Attribute=false, Bone Index=2, SkeletalMesh:NiagaraDataInterfaceSkeletalMesh=$User.SkeletalMesh)
   &SpawnRate(
    Spawn Probability=0.1
    SpawnRate=&Lerp_Float(
     A=10
     B=50
     Alpha=&FN_FrontEndFloatModifier(Gameplay=HlslExpression() {hlslExpression="(Engine.Owner.Velocity.z < -2600.0f) ? 1.0f : 0.0f"}, Front End=$User.ParaGlideLeanAlpha)
    )
    Use Spawn Probability=true
   ) {module=SpawnRate001}
  ]
  particleSpawnScript [
   /Niagara/Modules/Spawn/Initialization/V2/InitializeParticle(
    Lifetime Max=1.5
    Lifetime Mode=::NewEnumerator1
    Mass Mode=::NewEnumerator3
    Mesh Scale=&RandomRangeVector(Minimum=<1 1 1>, Maximum=<1.4 1.6 1>)
    Mesh Scale Mode=::NewEnumerator4
    Sprite Rotation Mode=::NewEnumerator2
    Sprite Size Mode=::NewEnumerator1
    Uniform Sprite Size Max=12
    Uniform Sprite Size Min=6
    UsePositionOffset=false
    Write Material Random=true
   )
   &FN_SetSpawnGroup(Number of Spawn Groups=2)
   &InitialMeshOrientation(
    Orientation Axis=<0 0 1>
    Orientation Vector=<0 -1 0>
    Rotation=&RandomRangeVector(Minimum=<0 0 0>, Maximum=<0 0 0>)
    Use Orientation Vector=true
    Use Rotation Vector=true
   )
   /SetVariables(Particles.SubImageIndex=&RandomRangeFloat(Minimum=0, Maximum=3)) {isSetParameters=true}
   &SkeletalMeshLocation(
    Bone / Socket Index=$Particles.SpawnGroup
    Bone Sampling Mode=::NewEnumerator4
    Inherit Velocity=false
    MaskSpawnGroup=false
    Skeletal Mesh=$User.SkeletalMesh
   )
   &SphereLocation(
    MaskSpawnGroup=false
    Offset=&MakeVector(X=&MaskFloatBySpawnGroup(Masked Float=-10, Passthrough Float=10, Spawn Group Mask=1), Y=0, Z=50)
    Offset Coordinate Space:ENiagaraCoordinateSpace=::World
    Override Seed=false
    Sphere Radius=32
    Surface Only=false
    UseNonUniformScale=false
    UseOffset=true
    UseSphereOrientation=false
   )
   &InheritVelocity(
    Module.Limit Speed=false
    Velocity Scale=&VectorFromFloat(Value=&RandomRangeFloat(Minimum=0.4, Maximum=0.7))
   )
   &AddVelocity(Velocity=&MakeVector(
     X=&RandomRangeFloat(Minimum=-200, Maximum=200)
     Y=&RandomRangeFloat(Minimum=-200, Maximum=200)
     Z=&FN_FrontEndFloatModifier(
      Gameplay=0
      Front End=&FloatFromCurve(
       FloatCurve=DataInterface() {propertyValues {Curve {keys [<at 0 1>,<at 1 2.5>]}}}
       CurveIndex=$User.ParaGlideLeanAlpha
       Scale Curve=&RandomRangeFloat(Minimum=200, Maximum=300)
      )
     )
    ))
  ]
  particleUpdateScript [
   &ParticleState()
   &GravityForce() {enabled=false}
   &Drag(Drag=&RandomRangeFloat(Minimum=0.25, Maximum=2), Use Linear Drag=true, Use Rotational Drag=false)
   &ScaleColor(
    Scale Alpha=&FloatFromCurve(FloatCurve=DataInterface() {propertyValues {Curve {keys [<at -0.0046 1.0113 RCIM_Cubic>,<at 1.0011 0.002 RCIM_Cubic>]}}})
    ScaleA=true
    ScaleRGB=true
   )
   &CurlNoiseForce(Noise Strength=300, Pan Noise=false)
   &SolveForcesAndVelocity(Clamp Velocity=false, Limit Acceleration=false)
   &UpdateMeshOrientation(
    Rotation Rate=&RandomRangeFloat(Minimum=-2, Maximum=2)
    Rotation Vector=&RandomVector(Vector Scale=1)
   )
   &ScaleMeshSize(Scale Factor=&VectorFromFloat(Value=&FloatFromCurve(FloatCurve=DataInterface() {propertyValues {Curve {keys [
          <at 0 0 RCIM_Cubic>
          {arriveTangent=-0.0752, interpMode=RCIM_Cubic, leaveTangent=-0.0752, tangentMode=RCTM_User, time=0.1, value=1}
          {arriveTangent=0.0346, interpMode=RCIM_Cubic, leaveTangent=0.0346, tangentMode=RCTM_User, time=0.5, value=1}
          <at 1 0 RCIM_Cubic>
         ]}}})))
  ]
  rendererClasses [&NiagaraMeshRendererProperties]
  renderers [{rendererClass=&NiagaraMeshRendererProperties, rendererIndex=0}]
 }
 LT_ClothPieces {
  bEnabled=false
  emitterUpdateScript [
   &EmitterState(Life Cycle Mode=::NewEnumerator1, Loop Duration=1, UseLoopDelay=false)
   &SpawnRate(
    Spawn Probability=0.1
    SpawnRate=HlslExpression() {hlslExpression="(Engine.Owner.Velocity.z < -2600.0f) ? 4.0f : 0.0f"}
    Use Spawn Probability=true
   )
  ]
  particleSpawnScript [
   /Niagara/Modules/Spawn/Initialization/InitializeParticle(
    Color=<rgba 0.013 0.0152 0.0242 1>
    Lifetime=&UniformRangedFloat(Minimum=1.8, Maximum=2)
    Mass=&UniformRangedFloat(Minimum=0.75, Maximum=1.75)
    Module.Write Color=true
    Module.Write Lifetime=true
    Module.Write Mass=true
    Module.Write Material Random=true
    Module.Write Position=false
    Module.Write Scale=false
    Module.Write SpriteRotation=false
    Module.Write SpriteSize=true
    Position=&SimulationPosition()
    Sprite Size=&UniformRangedVector2D(Minimum=<7 7>, Maximum=<7 15>)
   )
   /Game/Effects/Niagara/Modules/Groups/FN_SetSpawnGroup_v1/FN_SetSpawnGroup(Override SpawnGroup=true) {enabled=false}
   &FN_SetSpawnGroup_v2(Override SpawnGroup=true)
   /SetVariables(Particles.SubImageIndex=&UniformRangedFloat(Minimum=0, Maximum=3)) {isSetParameters=true}
   &SkeletalMeshSkeletonLocation(Module.Inherit Velocity=false, Module.Use Sampled Bone as Ribbon ID=true) {enabled=false}
   &SkeletalMeshLocation(
    Inherit Velocity=false
    MaskSpawnGroup=false
    Module.Override Randomness=false
    Skeletal Mesh=DataInterface() {propertyValues {
      DefaultMesh=null
      ExcludeBoneName=root
      OptionalComponentClass=null
      PreviewMesh=/Game/Characters/Player/Male/Medium/Bodies/M_Med_Soldier_04/Meshes/SK_M_Med_Soldier_04
      SkinningMode=SkinOnTheFly
      SoftSourceActor=null
      SourceMode=Default
      UvSetIndex=0
      WholeMeshLOD=-1
      bExcludeBone=false
      bReadDeformedGeometry=true
      bRequireCurrentFrameData=true
      ComponentTags []
      FilteredBones []
      FilteredSockets []
      MeshUserParameter {parameter {name=None, typeDefHandle {registeredTypeIndex=1}, varData []}}
      SamplingRegions []
     }}
   ) {enabled=false}
   &BoxLocation(Box Size=<7 1 0>, MaskSpawnGroup=false, Module.Override Seed=false, Rotate Box=false, Surface Only=false)
   &SkeletalMeshLocation(
    Bone Sampling Mode=::NewEnumerator0
    Inherit Velocity=false
    MaskSpawnGroup=false
    Module.Override Randomness=false
    Skeletal Mesh=DataInterface() {propertyValues {
      DefaultMesh=null
      ExcludeBoneName=root
      OptionalComponentClass=null
      PreviewMesh=&M_Med_Soldier_04_Preview
      SkinningMode=SkinOnTheFly
      SoftSourceActor=null
      SourceMode=Default
      UvSetIndex=0
      WholeMeshLOD=-1
      bExcludeBone=false
      bReadDeformedGeometry=true
      bRequireCurrentFrameData=true
      ComponentTags []
      FilteredBones [upperarm_l,upperarm_r,spine_02]
      FilteredSockets [LootEntryPosition]
      MeshUserParameter {parameter {name=None, typeDefHandle {registeredTypeIndex=1}, varData []}}
      SamplingRegions []
     }}
   ) {module=SkeletalMeshLocation001}
  ]
  particleUpdateScript [
   &ParticleState()
   &AccelerationForce(Acceleration=<0 0 2100>)
   /Niagara/Modules/Update/Forces/CurlNoiseForce(Noise Strength=250, Pan Noise=false) {module=CurlNoiseForce001}
   &SolveForcesAndVelocity(Clamp Velocity=false, Limit Acceleration=false)
   &ScaleSpriteSize(Scale Factor=&Vector2DFromCurve(Vector2Curve=DataInterface() {propertyValues {XCurve {keys [<at 0 1>,<at 1 0>]}, YCurve {keys [<at 0 1>,<at 1 0>]}}}, ScaleCurve=<1 1>))
  ]
  rendererClasses [&NiagaraSpriteRendererProperties]
  renderers [{rendererClass=&NiagaraSpriteRendererProperties}]
 }
}
summary {
 systemName=NS_Contrail_DinoHunter
 emitters [
  {bEnabled=false, emitterName=LT_ClothPieces, rendererClasses [&NiagaraSpriteRendererProperties]}
  {emitterName=Basic_Contrail, rendererClasses [&NiagaraRibbonRendererProperties]}
  {emitterName=Feathers, rendererClasses [&NiagaraMeshRendererProperties]}
  {emitterName=Capse_LocalSpace, rendererClasses [&NiagaraMeshRendererProperties]}
 ]
 userVariables [
  User.bIsFrontEnd {type=NiagaraBool}
  User.bIsFrontEndPreview {type=NiagaraBool}
  User.ParaGlideLeanAlpha {type=NiagaraFloat, defaultValue=0, description=""}
  User.PlayerSpeed {type=NiagaraFloat, defaultValue=0, description=""}
  User.RibbonIDOffset {type=NiagaraInt32}
  User.SkeletalMesh {
   type=NiagaraDataInterfaceSkeletalMesh
   defaultValue {
    struct=NiagaraExt_VariableValue_DataInterface
    value {dataInterface=&self:NiagaraDataInterfaceSkeletalMesh_1}
   }
  }
 ]
}
